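# Editor utility widget for previewing tree World Position Offset disable distances in the level viewport.
# Buttons cycle saved-index tree Blueprints as tagged transient actors at the origin, or rotate them 45 degrees.
# Preview derives the cutoff from root mesh distance and view-distance scale, jumps the camera inside it, and Tick glides toward any set target.
# Another button caps selected Blueprints' shadow proxy disable distances at their root's value through subobject data.
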EditorUtilityWidgetBlueprint /Game/NeverCook/EditorUtilities/TreeTools/TreeWPODistanceTester/EUW_TreeWPODistanceTester
graphs [
 EventGraph {nodes [
   "On Clicked (TA_Generic01_Button)"!() -> (OutputDelegate, then, widget:User Widget Object Reference, event:Pointer Event Structure) {is=title K2Node_ComponentBoundEvent}
   > RunPreview()
   "On Clicked (btnNextRot)"!() -> (OutputDelegate, then, widget:User Widget Object Reference, event:Pointer Event Structure) {is=title K2Node_ComponentBoundEvent}
   > n1 = getTreeActor!(execute, self) -> (then, OutActors:Array of BP Tree Asteria Parent Object References)
   > n2 = For Each Loop(Array:Array of BP Tree Asteria Parent Object References=@n1) -> (Array Element:BP Tree Asteria Parent Object Reference)
   > Add Actor World Rotation!(execute, self:Actor Object Reference=@n2, DeltaRotation:Rotator="0,45,0", bSweep:bool=false, bTeleport:bool=false) -> (then, SweepHitResult:Hit Result Structure)
   > RunPreview()
   "On Clicked (btnNextTree)"!() -> (OutputDelegate, then, widget:User Widget Object Reference, event:Pointer Event Structure) {is=title K2Node_ComponentBoundEvent}
   > n3 = getTreeActor()
   > n4 = Branch(Condition="integer > integer"(A=Length(TargetArray:Array of BP Tree Asteria Parent Object References=@n3)))
   > LoadNextTree()
   LoadCurrentTree(execute=@n4.else)
   "On Clicked (btnCopyShadowDist)"!() -> (OutputDelegate, then, widget:User Widget Object Reference, event:Pointer Event Structure) {is=title K2Node_ComponentBoundEvent}
   > UpdateShadowProxyDisableDistance()
   UpdateTreeListWithSelectedActors() {is=title K2Node_CustomEvent}
   > Execute Python Script(PythonScript="import unreal\r\n\r\nasset_class = unreal.load_class(None, \"&/EUW_TreeWPODistanceTester.EUW_TreeWPODistanceTester_C\")\r\nif asset_class is not None:\r\n    cdo = unreal.get_default_object(asset_class)\r\n    if cdo is not None:\r\n        tree_asset_list = []\r\n        selected_asset_datas = unreal.EditorUtilityLibrary.get_selected_assets()\r\n        cdo.set_editor_property(\"AssetLoadList\", selected_asset_datas)")
   n5 = Event Tick()
   > Branch(Condition="Not Equal Exactly (Vector)"!(A:Vector=Get FinalCameraLocation() -> (FinalCameraLocation:Vector), B:Vector="-1,-1,-1") -> (ReturnValue:bool))
   n6 = GetLevelViewportCameraInfo!(self:Unreal Editor Subsystem Object Reference=Get UnrealEditorSubsystem()) -> (CameraLocation:Vector, CameraRotation:Rotator, ReturnValue:bool)
   n7 = Get FinalCameraLocation()
   > SetLevelViewportCameraInfo!(
    execute
    self:Unreal Editor Subsystem Object Reference=Get UnrealEditorSubsystem()
    CameraLocation:Vector=VInterpTo_Constant!(Current:Vector=@n6, Target:Vector=@n7, DeltaTime:float=@n5.InDeltaTime, InterpSpeed:float=1200) -> (ReturnValue:Vector)
    CameraRotation:Rotator="0, 0, 0"
   ) -> (then)
   VInterpTo!(Current:Vector=@n6, Target:Vector=@n7, DeltaTime:float=@n5.InDeltaTime, InterpSpeed:float=1) -> (ReturnValue:Vector)
  ]}
 FocusTree_DisableDist {nodes [
   FocusTree_DisableDist() {is=title K2Node_FunctionEntry}
   > n1 = GetDisableLocation!(execute, self) -> (then, Location:Vector)
   n2 = Get CameraMoveDistance() -> (CameraMoveDistance:double)
   > Set FinalCameraLocation(FinalCameraLocation:Vector=vector - vector(A=@n1, B=MakeVector(X=@n2, Y=0))) -> (Output_Get:Vector)
  ]}
 GetCloseDistance {nodes [
   GetCloseDistance() {is=title K2Node_FunctionEntry}
   > n1 = GetDisableLocation()
   n2 = Get CameraMoveDistance()
   > Return Node!(execute, Output_Get:Vector=vector + vector(A=@n1, B=MakeVector(X=@n2, Y=0)))
  ]}
 GetDisableLocation {nodes [
   GetDisableLocation() {is=title K2Node_FunctionEntry}
   > n1 = getTreeActor()
   > n2 = Set l_targetActor(l_targetActor:BP Tree Asteria Parent Object Reference="Get (a copy)"(Array:Array of BP Tree Asteria Parent Object References=@n1) -> (Output:BP Tree Asteria Parent Object Reference)) -> (Output_Get:BP Tree Asteria Parent Object Reference)
   > n3 = Is Valid(InputObject=@n2)
   Return Node!(execute=@n3.Is Not Valid, Location:Vector="0, 0, 0")
   > Set l_boundsCenter(l_boundsCenter:Vector=Get Actor Bounds(self=Get l_targetActor() -> (l_targetActor:BP Tree Asteria Parent Object Reference), bOnlyCollidingComponents=true)) -> (Output_Get:Vector)
   > Set l_pivotLocation(l_pivotLocation:Vector=Get Actor Location(self=Get l_targetActor())) -> (Output_Get:Vector)
   > n4 = GetConsoleVariableFloatValue!(execute, VariableName:String=r.viewdistancescale) -> (then, ReturnValue:float)
   n5 = Get l_targetActor()
   > n6 = GetEditorProperty!(
    execute:Exec
    Object:Object Reference=Cast To StaticMeshComponent!(Object=Get RootComponent!(self:Actor Object Reference=@n5) -> (RootComponent:Scene Component Object Reference)) -> (AsStatic Mesh Component, bSuccess:bool=~)
    PropertyName:Name=WorldPositionOffsetDisableDistance
   ) -> (then:Exec, PropertyValue:int, ReturnValue:bool)
   > Set l_disableDistance(l_disableDistance:int=int * int(A=@n6, B=Ceil(A=@n4) {is=title K2Node_CallFunction})) -> (Output_Get:int=0)
   n7 = float - float(
    A=Get!() -> (l_boundsCenter_X:double, l_boundsCenter_Y:double, l_boundsCenter_Z:double).l_boundsCenter_Z
    B=Get!() -> (l_pivotLocation_X:double, l_pivotLocation_Y:double, l_pivotLocation_Z:double).l_pivotLocation_Z
   )
   n8 = MakeVector(X="To Float (Integer)"(InInt=Get l_disableDistance() -> (l_disableDistance:int=0)), Y=0)
   n9 = vector * vector(A=Get l_pivotLocation() -> (l_pivotLocation:Vector), B:Vector="1,1,0") -> (ReturnValue=~)
   > n10 = Set l_finalLoc(l_finalLoc:Vector=vector + vector(A=vector - vector(A=@n9, B=@n8), B=MakeVector(Y=0, Z=@n7))) -> (Output_Get:Vector)
   > Return Node(Location=@n10)
  ]}
 LoadCurrentTree {nodes [
   LoadCurrentTree() {is=title K2Node_FunctionEntry}
   > n1 = Sequence()
   > n2 = getTreeActor()
   > n3 = Reverse for Each Loop!(Exec:Exec, Array:Array of BP Tree Asteria Parent Object References=@n2) -> (LoopBody:Exec, ArrayIndex:int, ArrayElement:BP Tree Asteria Parent Object Reference, Completed:Exec) {is=title K2Node_MacroInstance}
   > DestroyActor!(execute, self:Editor Actor Subsystem Object Reference=Get EditorActorSubsystem(), ActorToDestroy:Actor Object Reference=@n3.ArrayElement) -> (then, ReturnValue:bool)
   n4 = SpawnActorFromObject!(
    execute=@n1.then_1
    self:Editor Actor Subsystem Object Reference=Get EditorActorSubsystem()
    ObjectToUse:Object Reference="Get (a copy)"(
     Array:Array of Object References=Get AssetLoadList() -> (AssetLoadList:Array of Object References)
     Dimension 1=Get currentAssetIndex() -> (currentAssetIndex:int=0)
    ) -> (Output:Object Reference)
    Location:Vector="0, 0, 0"
    Rotation:Rotator
    bTransient:bool=true
   ) -> (then, ReturnValue:Actor Object Reference)
   > Set Tags!(execute, Tags:Array of Names=Make Array("[0]":Name=SpawnedByTreeWPODistanceTest) -> (Array:Array of Names), self:Actor Object Reference=@n4) -> (then, Output_Get:Array of Names)
  ]}
 LoadNextTree {nodes [
   LoadNextTree() {is=title K2Node_FunctionEntry}
   > n1 = Sequence!(execute) -> (then_0, then_1, then_2:Exec)
   n2 = Increment Int(" "=@n1.then_1, Value=Get currentAssetIndex() -> (currentAssetIndex=0))
   n3 = Get AssetLoadList()
   > Branch(Condition="integer > integer"(A=@n2, B=Last Index!(TargetArray:Array of Object References=@n3) -> (ReturnValue=0)))
   > Set currentAssetIndex(currentAssetIndex:int=0) -> (Output_Get:int=0)
   SaveGame_Save(execute=@n1.then_2, self:TA Base Editor Widget Object Reference)
   > LoadCurrentTree() {is=title K2Node_CallFunction}
  ]}
 ResetLocation {nodes [
   ResetLocation() {is=title K2Node_FunctionEntry}
   > Set FinalCameraLocation(FinalCameraLocation="-1,-1,-1")
  ]}
 RunPreview {nodes [
   RunPreview() {is=title K2Node_FunctionEntry}
   > ResetLocation() {is=title K2Node_CallFunction}
   > n1 = GetCloseDistance!(execute, self) -> (then, Output_Get:Vector) {is=title K2Node_CallFunction}
   > SetLevelViewportCameraInfo(self=Get UnrealEditorSubsystem(), CameraLocation=@n1, CameraRotation="0, 0, 0")
   > Clear and Invalidate Timer by Handle!(execute, WorldContextObject:Object Reference, Handle=Get timerHandle_Focus() -> (timerHandle_Focus:Timer Handle Structure)) -> (then)
   > n2 = Set Timer by Event(Delegate=Create Event(), Time=0.5, bLooping=false)
   > Set timerHandle_Focus(timerHandle_Focus:Timer Handle Structure=@n2) -> (Output_Get:Timer Handle Structure)
  ]}
 SaveGame_PostLoad {nodes [
   n1 = SaveGame_PostLoad!() -> (then, saveGameExisted:bool, saveGame_Object:Save Game Object Reference) {is=title K2Node_FunctionEntry}
   > Branch(Condition=@n1)
   > n2 = Cast To TA_Base_SaveGame!(execute, Object=@n1.saveGame_Object) -> (then, CastFailed, AsTA Base Save Game:TA Base Save Game Object Reference) {is=title K2Node_DynamicCast}
   > Set currentAssetIndex(currentAssetIndex="Get (a copy)"(Array:Array of Integers=Get integers!(self:TA Base Save Game Object Reference=@n2) -> (integers:Array of Integers)) -> (Output:int)) -> (Output_Get=0)
   > LoadCurrentTree()
   Return Node!(execute, " ":bool=false)
  ]}
 SaveGame_PrepareToSave {nodes [
   n1 = SaveGame_PrepareToSave!() -> (then, SaveGameObject:Save Game Object Reference) {is=title K2Node_FunctionEntry}
   > n2 = Cast To TA_Base_SaveGame(Object=@n1)
   > Set integers!(
    execute
    integers:Array of Integers=Make Array("[0]":int=Get currentAssetIndex() -> (currentAssetIndex=0)) -> (Array:Array of Integers)
    self:TA Base Save Game Object Reference=@n2
   ) -> (then, Output_Get:Array of Integers)
   Return Node(" "=false)
  ]}
 UpdateShadowProxyDisableDistance {nodes [
   n1 = GetSelectedAssetData()
   > n2 = For Each Loop(Array:Array of Asset Data Structures=@n1) -> (Array Element:Asset Data Structure)
   n3 = Break Asset Data(AssetData=@n2)
   n4 = Append!(A="To String (Name)"(InName=@n3), B=., C:String="To String (Name)"(InName=@n3.AssetName), D:String=_C) -> (ReturnValue)
   > PrintString(InString=@n4)
   > n5 = Execute Python Script!(
    execute
    PythonScript="import unreal\r\n\r\nasset_class = unreal.load_class(None, asset_class_string)\r\ncdo = None\r\nif asset_class is not None:\r\n    cdo = unreal.get_default_object(asset_class)\r\n"
    asset_class_string:String=@n4
   ) -> (then, ReturnValue, cdo:Wildcard) {inputs [asset_class_string], outputs [cdo]}
   n6 = GetSelectedAssets(execute=@n20.then)
   > n7 = For Each Loop(Array:Array of Object References=@n6) -> (Array Element:Object Reference)
   n8 = GetBlueprintAsset(execute=[@n5.then, @n7.LoopBody], Object=@n7)
   > Is Valid(InputObject=@n8)
   > n9 = Gather Subobject Data For Blueprint(self=Get SubobjectDataSubsystem(), Context=@n8)
   > n10 = For Each Loop(Array:Array of Subobject Data Handle Structures=@n9) -> (Array Element:Subobject Data Handle Structure)
   > n11 = Cast To ShadowProxyMeshComponent!(
    execute
    Object=GetObject!(Data:"Subobject Data Structure (by ref)"=GetData(DataHandle=@n10), bEvenIfPendingKill:bool=false) -> (ReturnValue:Object Reference)
   ) -> (then, CastFailed, AsShadow Proxy Mesh Component:Shadow Proxy Mesh Component Object Reference) {is=title K2Node_DynamicCast}
   > n12 = GetEditorProperty(Object=@n11, PropertyName=WorldPositionOffsetDisableDistance)
   > Set l_proxyDist(l_proxyDist:int=@n12) -> (Output_Get:int=0)
   n13 = "Get (a ref)"(Array:Array of Subobject Data Handle Structures=@n9, Dimension 1=1) -> (Output:"Subobject Data Handle Structure (by ref)")
   > n14 = GetEditorProperty(Object=GetObject(Data=GetData(DataHandle=@n13), bEvenIfPendingKill=false), PropertyName=WorldPositionOffsetDisableDistance)
   > Set l_rootDist(l_rootDist:int=@n14) -> (Output_Get:int=0)
   n15 = Get l_proxyDist() -> (l_proxyDist:int=0)
   n16 = Get l_rootDist() -> (l_rootDist:int=0)
   n17 = "integer > integer"(A=@n15, B=@n16)
   > PrintString(InString=Append!(A="To String (Integer)"(InInt=@n15), B=" > ", C="To String (Integer)"(InInt=@n16), D=" = ", E:String="To String (Boolean)"(InBool=@n17)) -> (ReturnValue))
   > Branch(Condition=@n17)
   > n18 = Set l_proxyDist(l_proxyDist=Get l_rootDist() -> (l_rootDist=0)) -> (Output_Get=0)
   > n19 = SetEditorProperty(Object=@n11, PropertyName=WorldPositionOffsetDisableDistance, PropertyValue:int=@n18)
   > PrintString(InString=Append!(
     A="setting "
     B="To String (Integer)"(InInt=@n18)
     C=" on "
     D=GetDisplayName(Object=@n11)
     E=" == "
     F:String="To String (Boolean)"(InBool=@n19)
    ) -> (ReturnValue))
   n20 = UpdateShadowProxyDisableDistance() {is=title K2Node_FunctionEntry}
   "To String (Name)"(InName=@n3)
  ]}
 getTreeActor {nodes [
   getTreeActor() {is=title K2Node_FunctionEntry}
   > n1 = GetAllActorsOfClassWithTag!(
    execute
    WorldContextObject:Object Reference
    ActorClass:Actor Class Reference=/Game/Environments/Asteria/Foliage/Trees/Blueprint_Parent/BP_Tree_Asteria_Parent.BP_Tree_Asteria_Parent_C
    Tag:Name=SpawnedByTreeWPODistanceTest
   ) -> (then, OutActors:Array of BP Tree Asteria Parent Object References)
   > Return Node!(execute, OutActors:Array of BP Tree Asteria Parent Object References=@n1)
  ]}
]
variables [
 AssetLoadList {
  type="array<Object>"
  default [
   /Game/Environments/Asteria/Foliage/Trees/Birch/Blueprint/BP_Asteria_Birch_Tree
   /Game/Environments/Asteria/Foliage/Trees/CherryBlossom/Blueprint/BP_Asteria_Cherry_Blossom_Tree
   /Game/Environments/Asteria/Foliage/Trees/CherryBlossom/Blueprint/BP_Asteria_Cherry_Blossom_Tree_B
   /Game/Environments/Asteria/Foliage/Trees/DarkPine/Blueprint/BP_Asteria_DarkPine_2_Tree1
   /Game/Environments/Asteria/Foliage/Trees/DarkPine/Blueprint/BP_Asteria_DarkPine_Tree
   /Game/Environments/Asteria/Foliage/Trees/FallTree/Blueprint/BP_Asteria_FallTree_Tree
   /Game/Environments/Asteria/Foliage/Trees/JapaneseMaple/Blueprints/BP_Asteria_JapaneseMaple_Tree
   /Game/Environments/Asteria/Foliage/Trees/KapokTree/Blueprint/BP_Asteria_KapokTree_Branch_Parent
   /Game/Environments/Asteria/Foliage/Trees/Oak/Blueprint/BP_Asteria_Oak_Tree_Large
   /Game/Environments/Asteria/Foliage/Trees/Oak/Blueprint/BP_Asteria_Oak_Tree_Medium
   /Game/Environments/Asteria/Foliage/Trees/Palm/Blueprint/BP_Asteria_Palm_1_Tree
   /Game/Environments/Asteria/Foliage/Trees/Palm/Blueprint/BP_Asteria_Palm_2_Tree
   /Game/Environments/Asteria/Foliage/Trees/Pine/Blueprint/BP_Asteria_Pine_Small_Tree
   /Game/Environments/Asteria/Foliage/Trees/Pine/Blueprint/BP_Asteria_PIne_Tree_1
   /Game/Environments/Asteria/Foliage/Trees/Bamboo/Blueprint/BP_BambooTree_A
   /Game/Environments/Asteria/Foliage/Trees/Bamboo/Blueprint/BP_BambooTree_ClusterA_128
   /Game/Environments/Asteria/Foliage/Trees/Bamboo/Blueprint/BP_BambooTree_ClusterB_256
   /Game/Environments/Asteria/Foliage/Trees/Bamboo/Blueprint/BP_BambooTree_ClusterC_512
   /Game/Environments/Asteria/Foliage/Trees/BalsaTree/Blueprint/BP_Tree_Asteria_BalsaTree
   /Game/Environments/Asteria/Foliage/Trees/Blueprint_Parent/BP_Tree_Asteria_Parent_Vines
   /Game/Environments/Asteria/Foliage/Trees/RubberTree/Blueprint/BP_Tree_Asteria_RubberTree
   /Game/Environments/Asteria/Foliage/Trees/SnowyPine/Blueprint/BP_Tree_Asteria_SnowyPine
   /Game/Environments/Asteria/Foliage/Trees/SnowyPine/Blueprint/BP_Tree_Asteria_SnowyPine_Small
   /Game/Environments/Asteria/Foliage/Trees/UmbrellaTree/Blueprint/BP_Umbrella_Tree
  ]
 }
 CameraMoveDistance {type=number, default=1500}
 currentAssetIndex {type=integer}
 FinalCameraLocation {type=Vector, default=<-1 -1 -1>}
 isClose {type=boolean}
 timerHandle_Focus {type=TimerHandle}
]
widgets {
 info {
  parentClass=/Game/Blueprints/Blutilities/ForTechArt/Widgets/TA_Base_EditorWidget.TA_Base_EditorWidget_C
  rootWidgetClass=&StackBox
  widgetCount=5
 }
 widgets [
  StackBox_0 {widgetClassPath=&StackBox, widget=StackBox_0, properties {orientation=Orient_Vertical}}
  TA_Generic01_Button {
   widgetClassPath=&TA_Generic01_Button
   parent=StackBox_0
   widget=TA_Generic01_Button
   properties {label=Toggle Camera Location}
   slot {class=StackBoxSlot}
  }
  btnNextRot {
   widgetClassPath=&TA_Generic01_Button
   parent=StackBox_0
   widget=btnNextRot
   properties {label=Rotate 45 and Preview}
   slot {class=StackBoxSlot}
  }
  btnCopyShadowDist {
   widgetClassPath=&TA_Generic01_Button
   parent=StackBox_0
   widget=btnCopyShadowDist
   properties {label=CopyRootDistanceToShadowProxies}
   slot {class=StackBoxSlot, properties {verticalAlignment=VAlign_Bottom, size {sizeRule=Fill, value=1}}}
  }
  btnNextTree {
   widgetClassPath=&TA_Generic01_Button
   parent=StackBox_0
   widget=btnNextTree
   properties {label=Load Next Tree}
   slot {class=StackBoxSlot, properties {verticalAlignment=VAlign_Bottom, size {sizeRule=Fill, value=1}}}
  }
 ]
}
